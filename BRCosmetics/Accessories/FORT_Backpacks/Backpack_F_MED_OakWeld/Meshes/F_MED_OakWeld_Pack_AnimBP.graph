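# Post-process style AnimBlueprint for the OakWeld backpack, posing the pack and simulating its wings on the character.
# Blends a pose copied from a character part mesh (cached on customization load) over the input, then additively offsets the pack and upper wing roots, mostly gated by bIsSkyDiving.
# Chained spring-based AnimDynamics then swing upper and lower wing bones; both Up_05 nodes are unwired.
# EventGraph derives Parashuting and SkyDaving values from spine and pelvis yaw.
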AnimBlueprint /BRCosmetics/Accessories/FORT_Backpacks/Backpack_F_MED_OakWeld/Meshes/F_MED_OakWeld_Pack_AnimBP
target_skeleton=&/F_MED_OakWeld_Pack_Skeleton
graphs [
 AnimGraph {nodes [
   Anim Dynamics - Bone: Ving_R_Up_05(Alpha=1, ExternalForce="0,0,0") {
    angularDampingOverride=0.9999
    angularSpringConstant=800
    bAngularSpring=true
    bOverrideAngularDamping=true
    bUseGravityOverride=true
    simulationSpace=RootRelative
    alphaBoolBlend {blendOption=Linear}
    boundBone {boneName=Ving_R_Up_05}
    componentAppliedLinearAccClamp=<1200 1200 1200>
    componentLinearAccScale=<1 1 1>
    componentLinearVelScale=<1 1 1>
    gravityOverride {z=-980}
    physicsBodyDefinitions [{
      boundBone {boneName=Ving_R_Up_05}
      boxExtents=<9 9 9>
      constraintSetup {angularLimitsMax=<0 5 5>, angularLimitsMin=<0 -7 0>, angularTarget=<1 0 0>}
      localJointOffset=<1 0 0>
     }]
   }
   Anim Dynamics - Bone: Ving_L_Up_05(Alpha=1, ExternalForce="0,0,0") {
    angularDampingOverride=0.9999
    angularSpringConstant=1200
    bAngularSpring=true
    bOverrideAngularDamping=true
    bUseGravityOverride=true
    simulationSpace=RootRelative
    alphaBoolBlend {blendOption=Linear}
    boundBone {boneName=Ving_L_Up_05}
    componentAppliedLinearAccClamp=<1000 1000 1000>
    componentLinearAccScale=<0.8 0.8 0.8>
    componentLinearVelScale=<0.8 0.8 0.8>
    gravityOverride {z=-980}
    physicsBodyDefinitions [{
      boundBone {boneName=Ving_L_Up_05}
      boxExtents=<11 11 11>
      constraintSetup {angularLimitsMax=<0 5 5>, angularLimitsMin=<0 -7 0>, angularTarget=<1 0 0>}
      localJointOffset=<0.9 0 0>
     }]
   }
   n1 = Copy Pose From Mesh(SourceMeshComponent=Get Output() -> (Output:Skeletal Mesh Component Budgeted Object Reference))
   n2 = Input Pose()
   n3 = "Transform (Modify) Bone - Bone: pack"(ComponentPose=Local To Component(LocalPose=Layered blend per bone(BasePose=@n2, BlendPoses_0=@n1)), Translation="0,-0.5,5", Rotation="0,0,0") {
    rotationMode=BMM_Additive
    scaleMode=BMM_Additive
    translationMode=BMM_Additive
    alphaBoolBlend {blendOption=Linear}
    boneToModify {boneName=pack}
    translation {y=-0.5, z=5}
   }
   n4 = "Transform (Modify) Bone - Bone: pack"!(ComponentPose=@n3, bAlphaBoolEnabled:bool=Get bIsSkyDiving(), Translation="0,-1,0", Rotation="0,0,-5", Scale) -> (Pose) {
    alphaInputType=Bool
    rotationMode=BMM_Additive
    scaleMode=BMM_Additive
    translationMode=BMM_Additive
    alphaBoolBlend {blendOption=Linear}
    boneToModify {boneName=pack}
    rotation {roll=-5}
    translation {y=-1}
   }
   n5 = "Transform (Modify) Bone - Bone: pack"!(ComponentPose=@n4, Alpha:float, Translation="0,-3,0", Rotation="0,0,25", Scale) -> (Pose) {
    rotationMode=BMM_Additive
    scaleMode=BMM_Additive
    translationMode=BMM_Additive
    alphaBoolBlend {blendOption=Linear}
    boneToModify {boneName=pack}
    rotation {roll=25}
    translation {y=-3}
   }
   n6 = Get bIsSkyDiving()
   n7 = "Transform (Modify) Bone - Bone: Ving_L_Up_01"!(ComponentPose=@n5, bAlphaBoolEnabled:bool=@n6, Translation="0,0,0", Rotation="0,15,0", Scale="1,1,1") -> (Pose) {
    alphaInputType=Bool
    rotationMode=BMM_Additive
    scaleMode=BMM_Additive
    translationMode=BMM_Additive
    alphaBoolBlend {blendOption=Linear}
    boneToModify {boneName=Ving_L_Up_01}
    rotation {yaw=15}
   }
   n8 = "Transform (Modify) Bone - Bone: Ving_R_Up_01"!(ComponentPose=@n7, bAlphaBoolEnabled:bool=@n6, Translation="0,0,0", Rotation="0,-15,0", Scale="1,1,1") -> (Pose) {
    alphaInputType=Bool
    rotationMode=BMM_Additive
    scaleMode=BMM_Additive
    translationMode=BMM_Additive
    alphaBoolBlend {blendOption=Linear}
    boneToModify {boneName=Ving_R_Up_01}
    rotation {yaw=-15}
   }
   n9 = "Transform (Modify) Bone - Bone: pack"!(ComponentPose=@n8, bAlphaBoolEnabled:bool=@n6, Translation="0,-2,0", Rotation="0,0,0", Scale) -> (Pose) {
    alphaInputType=Bool
    scaleMode=BMM_Additive
    translationMode=BMM_Additive
    alphaBoolBlend {blendOption=Linear}
    boneToModify {boneName=pack}
    translation {y=-2}
   }
   n10 = Anim Dynamics - Bone: Ving_L_Up_01(ComponentPose=@n9, Alpha=1, ExternalForce="0,0,0") {
    angularDampingOverride=0.999
    angularSpringConstant=1200
    bAngularSpring=true
    bOverrideAngularDamping=true
    bUseGravityOverride=true
    simulationSpace=RootRelative
    alphaBoolBlend {blendOption=Linear}
    boundBone {boneName=Ving_L_Up_01}
    componentAppliedLinearAccClamp=<900 900 900>
    componentLinearAccScale=<0.7 0.7 0.7>
    componentLinearVelScale=<0.7 0.7 0.7>
    gravityOverride {z=-980}
    physicsBodyDefinitions [{
      boundBone {boneName=Ving_L_Up_01}
      boxExtents=<11 11 11>
      constraintSetup {angularLimitsMax=<0 8 20>, angularLimitsMin=<0 -4 0>, angularTarget=<1 0 0>}
      localJointOffset=<0.8 0 0>
     }]
   }
   n11 = Anim Dynamics - Bone: Ving_L_Up_02(ComponentPose=@n10, Alpha=1, ExternalForce="0,0,0") {
    angularDampingOverride=0.999
    angularSpringConstant=1200
    bAngularSpring=true
    bOverrideAngularDamping=true
    bUseGravityOverride=true
    simulationSpace=RootRelative
    alphaBoolBlend {blendOption=Linear}
    boundBone {boneName=Ving_L_Up_02}
    componentAppliedLinearAccClamp=<900 900 900>
    componentLinearAccScale=<0.7 0.7 0.7>
    componentLinearVelScale=<0.7 0.7 0.7>
    gravityOverride {z=-980}
    physicsBodyDefinitions [{
      boundBone {boneName=Ving_L_Up_02}
      boxExtents=<11 11 11>
      constraintSetup {angularLimitsMax=<0 8 14>, angularLimitsMin=<0 -8 0>, angularTarget=<1 0 0>}
      localJointOffset=<0.8 0 0>
     }]
   }
   n12 = Anim Dynamics - Bone: Ving_L_Up_03(ComponentPose=@n11, Alpha=1, ExternalForce="0,0,0") {
    angularDampingOverride=0.999
    angularSpringConstant=1200
    bAngularSpring=true
    bOverrideAngularDamping=true
    bUseGravityOverride=true
    simulationSpace=RootRelative
    alphaBoolBlend {blendOption=Linear}
    boundBone {boneName=Ving_L_Up_03}
    componentAppliedLinearAccClamp=<900 900 900>
    componentLinearAccScale=<0.7 0.7 0.7>
    componentLinearVelScale=<0.7 0.7 0.7>
    gravityOverride {z=-980}
    physicsBodyDefinitions [{
      boundBone {boneName=Ving_L_Up_03}
      boxExtents=<11 11 11>
      constraintSetup {angularLimitsMax=<0 6 9>, angularLimitsMin=<0 -8 0>, angularTarget=<1 0 0>}
      localJointOffset=<0.8 0 0>
     }]
   }
   n13 = Anim Dynamics - Bone: Ving_L_Up_04(ComponentPose=@n12, Alpha=1, ExternalForce="0,0,0") {
    angularDampingOverride=0.999
    angularSpringConstant=1300
    bAngularSpring=true
    bOverrideAngularDamping=true
    bUseGravityOverride=true
    simulationSpace=RootRelative
    alphaBoolBlend {blendOption=Linear}
    boundBone {boneName=Ving_L_Up_04}
    componentAppliedLinearAccClamp=<900 900 900>
    componentLinearAccScale=<0.7 0.7 0.7>
    componentLinearVelScale=<0.7 0.7 0.7>
    gravityOverride {z=-980}
    physicsBodyDefinitions [{
      boundBone {boneName=Ving_L_Up_04}
      boxExtents=<12 12 12>
      constraintSetup {angularLimitsMax=<0 5 6>, angularLimitsMin=<0 -7 0>, angularTarget=<1 0 0>}
      localJointOffset=<0.8 0 0>
     }]
   }
   n14 = Anim Dynamics - Bone: Ving_R_Up_01(ComponentPose=@n13, Alpha=1, ExternalForce="0,0,0") {
    angularDampingOverride=0.999
    angularSpringConstant=900
    bAngularSpring=true
    bOverrideAngularDamping=true
    bUseGravityOverride=true
    simulationSpace=RootRelative
    alphaBoolBlend {blendOption=Linear}
    boundBone {boneName=Ving_R_Up_01}
    componentAppliedLinearAccClamp=<1100 1100 1100>
    componentLinearAccScale=<0.9 0.9 0.9>
    componentLinearVelScale=<0.9 0.9 0.9>
    gravityOverride {z=-980}
    physicsBodyDefinitions [{
      boundBone {boneName=Ving_R_Up_01}
      boxExtents=<10 10 10>
      constraintSetup {angularLimitsMax=<0 8 23>, angularLimitsMin=<0 -4 0>, angularTarget=<1 0 0>}
      localJointOffset=<0.9 0 0>
     }]
   }
   n15 = Anim Dynamics - Bone: Ving_R_Up_02(ComponentPose=@n14, Alpha=1, ExternalForce="0,0,0") {
    angularDampingOverride=0.999
    angularSpringConstant=900
    bAngularSpring=true
    bOverrideAngularDamping=true
    bUseGravityOverride=true
    simulationSpace=RootRelative
    alphaBoolBlend {blendOption=Linear}
    boundBone {boneName=Ving_R_Up_02}
    componentAppliedLinearAccClamp=<1100 1100 1100>
    componentLinearAccScale=<0.9 0.9 0.9>
    componentLinearVelScale=<0.9 0.9 0.9>
    gravityOverride {z=-980}
    physicsBodyDefinitions [{
      boundBone {boneName=Ving_R_Up_02}
      boxExtents=<10 10 10>
      constraintSetup {angularLimitsMax=<0 8 18>, angularLimitsMin=<0 -8 0>, angularTarget=<1 0 0>}
      localJointOffset=<0.9 0 0>
     }]
   }
   n16 = Anim Dynamics - Bone: Ving_R_Up_03(ComponentPose=@n15, Alpha=1, ExternalForce="0,0,0") {
    angularDampingOverride=0.999
    angularSpringConstant=900
    bAngularSpring=true
    bOverrideAngularDamping=true
    bUseGravityOverride=true
    simulationSpace=RootRelative
    alphaBoolBlend {blendOption=Linear}
    boundBone {boneName=Ving_R_Up_03}
    componentAppliedLinearAccClamp=<1100 1100 1100>
    componentLinearAccScale=<0.9 0.9 0.9>
    componentLinearVelScale=<0.9 0.9 0.9>
    gravityOverride {z=-980}
    physicsBodyDefinitions [{
      boundBone {boneName=Ving_R_Up_03}
      boxExtents=<10 10 10>
      constraintSetup {angularLimitsMax=<0 6 9>, angularLimitsMin=<0 -8 0>, angularTarget=<1 0 0>}
      localJointOffset=<0.9 0 0>
     }]
   }
   n17 = Anim Dynamics - Bone: Ving_R_Up_04(ComponentPose=@n16, Alpha=1, ExternalForce="0,0,0") {
    angularDampingOverride=0.999
    angularSpringConstant=900
    bAngularSpring=true
    bOverrideAngularDamping=true
    bUseGravityOverride=true
    simulationSpace=RootRelative
    alphaBoolBlend {blendOption=Linear}
    boundBone {boneName=Ving_R_Up_04}
    componentAppliedLinearAccClamp=<1100 1100 1100>
    componentLinearAccScale=<0.9 0.9 0.9>
    componentLinearVelScale=<0.9 0.9 0.9>
    gravityOverride {z=-980}
    physicsBodyDefinitions [{
      boundBone {boneName=Ving_R_Up_04}
      boxExtents=<10 10 10>
      constraintSetup {angularLimitsMax=<0 5 6>, angularLimitsMin=<0 -7 0>, angularTarget=<1 0 0>}
      localJointOffset=<0.9 0 0>
     }]
   }
   n18 = Anim Dynamics - Bone: Ving_L_Down_01(ComponentPose=@n17, Alpha=1, ExternalForce="0,0,0") {
    angularDampingOverride=0.999
    angularSpringConstant=1000
    bAngularSpring=true
    bOverrideAngularDamping=true
    bUseGravityOverride=true
    simulationSpace=RootRelative
    alphaBoolBlend {blendOption=Linear}
    boundBone {boneName=Ving_L_Down_01}
    componentAppliedLinearAccClamp=<900 900 900>
    componentLinearAccScale=<0.8 0.8 0.8>
    componentLinearVelScale=<0.8 0.8 0.8>
    gravityOverride {z=-980}
    physicsBodyDefinitions [{
      boundBone {boneName=Ving_L_Down_01}
      boxExtents=<10 10 10>
      constraintSetup {angularLimitsMax=<0 3 8>, angularLimitsMin=<0 -8 0>, angularTarget=<1 0 0>}
      localJointOffset=<0.5 0 0>
     }]
   }
   n19 = Anim Dynamics - Bone: Ving_L_Down_02(ComponentPose=@n18, Alpha=1, ExternalForce="0,0,0") {
    angularDampingOverride=0.999
    angularSpringConstant=1000
    bAngularSpring=true
    bOverrideAngularDamping=true
    bUseGravityOverride=true
    simulationSpace=RootRelative
    alphaBoolBlend {blendOption=Linear}
    boundBone {boneName=Ving_L_Down_02}
    componentAppliedLinearAccClamp=<900 900 900>
    componentLinearAccScale=<0.8 0.8 0.8>
    componentLinearVelScale=<0.8 0.8 0.8>
    gravityOverride {z=-980}
    physicsBodyDefinitions [{
      boundBone {boneName=Ving_L_Down_02}
      boxExtents=<10 10 10>
      constraintSetup {angularLimitsMax=<0 3 8>, angularLimitsMin=<0 -8 0>, angularTarget=<1 0 0>}
      localJointOffset=<0.5 0 0>
     }]
   }
   n20 = Anim Dynamics - Bone: Ving_L_Down_03(ComponentPose=@n19, Alpha=1, ExternalForce="0,0,0") {
    angularDampingOverride=0.999
    angularSpringConstant=1000
    bAngularSpring=true
    bOverrideAngularDamping=true
    bUseGravityOverride=true
    simulationSpace=RootRelative
    alphaBoolBlend {blendOption=Linear}
    boundBone {boneName=Ving_L_Down_03}
    componentAppliedLinearAccClamp=<900 900 900>
    componentLinearAccScale=<0.8 0.8 0.8>
    componentLinearVelScale=<0.8 0.8 0.8>
    gravityOverride {z=-980}
    physicsBodyDefinitions [{
      boundBone {boneName=Ving_L_Down_03}
      boxExtents=<10 10 10>
      constraintSetup {angularLimitsMax=<0 3 8>, angularLimitsMin=<0 -8 0>, angularTarget=<1 0 0>}
      localJointOffset=<0.5 0 0>
     }]
   }
   n21 = Anim Dynamics - Bone: Ving_R_Down_01(ComponentPose=@n20, Alpha=1, ExternalForce="0,0,0") {
    angularDampingOverride=0.999
    angularSpringConstant=900
    bAngularSpring=true
    bOverrideAngularDamping=true
    bUseGravityOverride=true
    simulationSpace=RootRelative
    alphaBoolBlend {blendOption=Linear}
    boundBone {boneName=Ving_R_Down_01}
    componentAppliedLinearAccClamp=<1000 1000 1000>
    componentLinearAccScale=<0.9 0.9 0.9>
    componentLinearVelScale=<0.9 0.9 0.9>
    gravityOverride {z=-980}
    physicsBodyDefinitions [{
      boundBone {boneName=Ving_R_Down_01}
      boxExtents=<10 10 10>
      constraintSetup {angularLimitsMax=<0 7 11>, angularLimitsMin=<0 -3 0>, angularTarget=<1 0 0>}
      localJointOffset=<0.4 0 0>
     }]
   }
   n22 = Anim Dynamics - Bone: Ving_R_Down_02(ComponentPose=@n21, Alpha=1, ExternalForce="0,0,0") {
    angularDampingOverride=0.999
    angularSpringConstant=900
    bAngularSpring=true
    bOverrideAngularDamping=true
    bUseGravityOverride=true
    simulationSpace=RootRelative
    alphaBoolBlend {blendOption=Linear}
    boundBone {boneName=Ving_R_Down_02}
    componentAppliedLinearAccClamp=<1000 1000 1000>
    componentLinearAccScale=<0.9 0.9 0.9>
    componentLinearVelScale=<0.9 0.9 0.9>
    gravityOverride {z=-980}
    physicsBodyDefinitions [{
      boundBone {boneName=Ving_R_Down_02}
      boxExtents=<10 10 10>
      constraintSetup {angularLimitsMax=<0 7 11>, angularLimitsMin=<0 -3 0>, angularTarget=<1 0 0>}
      localJointOffset=<0.4 0 0>
     }]
   }
   n23 = Anim Dynamics - Bone: Ving_R_Down_03(ComponentPose=@n22, Alpha=1, ExternalForce="0,0,0") {
    angularDampingOverride=0.999
    angularSpringConstant=900
    bAngularSpring=true
    bOverrideAngularDamping=true
    bUseGravityOverride=true
    simulationSpace=RootRelative
    alphaBoolBlend {blendOption=Linear}
    boundBone {boneName=Ving_R_Down_03}
    componentAppliedLinearAccClamp=<1000 1000 1000>
    componentLinearAccScale=<0.9 0.9 0.9>
    componentLinearVelScale=<0.9 0.9 0.9>
    gravityOverride {z=-980}
    physicsBodyDefinitions [{
      boundBone {boneName=Ving_R_Down_03}
      boxExtents=<10 10 10>
      constraintSetup {angularLimitsMax=<0 7 11>, angularLimitsMin=<0 -3 0>, angularTarget=<1 0 0>}
      localJointOffset=<0.4 0 0>
     }]
   }
   Output Pose(Result=Component To Local(ComponentPose=@n23)) {layerGroup=None}
   Get SkyDaving() -> (SkyDaving:double)
  ]}
 EventGraph {nodes [
   Event BlueprintUpdateAnimation()
   n1 = GetOwnerSkeletalMeshComponent()
   n2 = GetDeltaTransformFromRefPose!(self=@n1, BoneName=spine_05, BaseName=spine_01) -> (ReturnValue_Location:Vector, ReturnValue_Rotation:Rotator, ReturnValue_Scale:Vector)
   > Set Parashuting(Parashuting:double=MapRangeClamped(Value=BreakRotator(InRot=@n2.ReturnValue_Rotation).Yaw, InRangeA=-15, InRangeB=-50)) -> (Output_Get:double)
   n3 = Event OnCharacterCustomizationLoaded()
   > Set Output(Output:Skeletal Mesh Component Budgeted Object Reference=Get Array Item with EFort Custom Part Type(Array=@n3.CustomMeshParts)) -> (Output_Get:Skeletal Mesh Component Budgeted Object Reference)
   Branch(Condition=Get bAnimDynamicsEnabled() -> (bAnimDynamicsEnabled:bool))
   n4 = BreakRotator(InRot=GetDeltaTransformFromRefPose(self=@n1, BoneName=pelvis, BaseName=root).ReturnValue_Rotation)
   n5 = MapRangeClamped(Value=@n4.Yaw, InRangeA=-140, InRangeB=-170, OutRangeB=0.4)
   n6 = MapRangeClamped(Value=@n4.Yaw, InRangeA=90, InRangeB=130)
   > Set SkyDaving(SkyDaving:double="Absolute (Float)"(A=float + float(A=@n6, B=@n5))) -> (Output_Get:double)
   TryGetPawnOwner()
  ]}
]
variables [
 Delta Time X {type=number}
 Is Skydive Floating {type=boolean}
 NewVar {type=number}
 NewVar_0 {type=number}
 Output {type=SkeletalMeshComponentBudgeted}
 Parashuting {type=number}
 SkyDaving {type=number}
]
